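# Physics asset for the dynamic collar of the F_MED_Raptor female player cosmetic body.
# It puts capsule bodies on three scarf bones (back, left, right) and a smaller capsule on spine_05.
# Each scarf body is constrained directly to spine_05 with angular limits, allowing the back piece wider swing but tighter twist than the two sides.
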PhysicsAsset /BRCosmetics/Characters/Player/Female/Medium/Bodies/F_MED_Raptor/Meshes/F_MED_Raptor_Dyn_Collar
bodies {
 dyn_parent_scarf_b [{shapeType=Capsule, length=4.755, radius=4.25, center=<0 0 0.4832>, rotation=<rot -15 0 0>}]
 dyn_parent_scarf_l [{shapeType=Capsule, length=3.755, radius=3.75, center=<0 0 0.1>, rotation=<rot 0 -90 0>}]
 dyn_parent_scarf_r [{shapeType=Capsule, length=3.755, radius=3.75, center=<0 0 0.1>, rotation=<rot 0 -90 0>}]
 spine_05 [{shapeType=Capsule, length=4.1436, radius=0.8186, center=<-2.3712 -1.868 -0.0156>, rotation=<rot 0.1796 -74.5533 -89.9999>}]
}
constraints [
 {bone1Name=dyn_parent_scarf_b, bone2Name=spine_05, swing1LimitDegrees=80, swing2LimitDegrees=80, twistLimitDegrees=25}
 {bone1Name=dyn_parent_scarf_l, bone2Name=spine_05, swing1LimitDegrees=45, swing2LimitDegrees=45, twistLimitDegrees=45}
 {bone1Name=dyn_parent_scarf_r, bone2Name=spine_05, swing1LimitDegrees=45, swing2LimitDegrees=45, twistLimitDegrees=45}
]
