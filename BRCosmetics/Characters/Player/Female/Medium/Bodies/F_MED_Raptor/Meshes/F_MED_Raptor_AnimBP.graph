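# Post-pose Anim Blueprint for the F_MED_Raptor female body cosmetic, copying the attached parent mesh's pose and adding secondary motion.
# Crouch events toggle additive bullet-bone offsets, while per-frame lowerarm rotation deltas from ref pose drive upperarm cuff corrections.
# AnimDynamics sways scarf, tags and boot strings, with tags first constrained between two spine bones.
# Two RigidBody nodes then simulate jacket and collar in spine_01 space using dedicated physics assets.
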AnimBlueprint /BRCosmetics/Characters/Player/Female/Medium/Bodies/F_MED_Raptor/Meshes/F_MED_Raptor_AnimBP
target_skeleton=&/F_MED_Raptor_Skeleton
graphs [
 AnimGraph {nodes [
   Anim Dynamics - Bone: dyn_scarf_r(Alpha=0.25, ExternalForce="0,0,0") {
    bGravityOverrideInSimSpace=true
    gravityScale=0.1
    lODThreshold=1
    alphaBoolBlend {blendOption=Linear}
    alphaScaleBiasClamp {bClampResult=true, clampMax=0.33, clampMin=0.1, scale=0.0033}
    boundBone {boneName=dyn_scarf_r}
    physicsBodyDefinitions [{
      sphereCollisionRadius=0
      boundBone {boneName=dyn_scarf_r}
      boxExtents=<2 2 2>
      constraintSetup {
       angularTargetAxis=AxisZ
       twistAxis=AxisZ
       angularLimitsMax=<0 0 0>
       angularLimitsMin=<0 0 0>
       angularTarget=<0 0 0>
       linearAxesMax=<2 2 2>
       linearAxesMin=<-2 -2 -2>
      }
      localJointOffset=<0 0 -1>
     }]
   }
   n1 = Get crouch_alpha() -> (crouch_alpha:double)
   n2 = "Transform (Modify) Bone - Bone: dyn_bullet_03"(
    ComponentPose=Local To Component(LocalPose=Copy Pose From Mesh() {bUseAttachedParent=true})
    Alpha=@n1
    Translation="-0.000,2.55,0.000"
    Rotation="0,0,0"
    Scale="1,1,1"
   ) {
    translationMode=BMM_Additive
    translationSpace=BCS_BoneSpace
    alphaBoolBlend {blendOption=Linear}
    boneToModify {boneName=dyn_bullet_03}
    translation {y=2.55}
   }
   n3 = "Transform (Modify) Bone - Bone: dyn_bullet_02"(ComponentPose=@n2, Alpha=@n1, Translation="-0.000,2.5,0.000", Rotation="0,0,0", Scale="1,1,1") {
    translationMode=BMM_Additive
    translationSpace=BCS_BoneSpace
    alphaBoolBlend {blendOption=Linear}
    boneToModify {boneName=dyn_bullet_02}
    translation {y=2.5}
   }
   n4 = "Transform (Modify) Bone - Bone: dyn_bullet_01"(ComponentPose=@n3, Alpha=@n1, Translation="-0.000,2.5,0.000", Rotation="0,0,0", Scale="1,1,1") {
    translationMode=BMM_Additive
    translationSpace=BCS_BoneSpace
    alphaBoolBlend {blendOption=Linear}
    boneToModify {boneName=dyn_bullet_01}
    translation {y=2.5}
   }
   n5 = Anim Dynamics - Bone: dyn_scarf_r(ComponentPose=@n4, Alpha=0.25, ExternalForce="0,0,0") {
    bGravityOverrideInSimSpace=true
    gravityScale=0.1
    lODThreshold=1
    alphaBoolBlend {blendOption=Linear}
    alphaScaleBiasClamp {bClampResult=true, clampMax=0.33, clampMin=0.1, scale=0.0033}
    boundBone {boneName=dyn_scarf_r}
    physicsBodyDefinitions [{
      sphereCollisionRadius=0
      boundBone {boneName=dyn_scarf_r}
      boxExtents=<2 2 2>
      constraintSetup {
       angularTargetAxis=AxisZ
       twistAxis=AxisZ
       angularLimitsMax=<0 0 0>
       angularLimitsMin=<0 0 0>
       angularTarget=<0 0 0>
       linearAxesMax=<2 2 2>
       linearAxesMin=<-2 -2 -2>
      }
      localJointOffset=<0 0 -1>
     }]
   }
   n6 = Anim Dynamics - Bone: dyn_scarf_l(ComponentPose=@n5, Alpha=0.25, ExternalForce="0,0,0") {
    bGravityOverrideInSimSpace=true
    gravityScale=0.1
    lODThreshold=1
    alphaBoolBlend {blendOption=Linear}
    alphaScaleBiasClamp {bClampResult=true, clampMax=0.33, clampMin=0.1, scale=0.0033}
    boundBone {boneName=dyn_scarf_l}
    physicsBodyDefinitions [{
      sphereCollisionRadius=0
      boundBone {boneName=dyn_scarf_l}
      boxExtents=<2 2 2>
      constraintSetup {
       angularTargetAxis=AxisZ
       twistAxis=AxisZ
       angularLimitsMax=<0 0 0>
       angularLimitsMin=<0 0 0>
       angularTarget=<0 0 0>
       linearAxesMax=<2 2 2>
       linearAxesMin=<-2 -2 -2>
      }
      localJointOffset=<0 0 -1>
     }]
   }
   n7 = Anim Dynamics - Bone: dyn_scarf_b(ComponentPose=@n6, Alpha=0.25, ExternalForce="0,0,0") {
    bGravityOverrideInSimSpace=true
    gravityScale=0.33
    lODThreshold=1
    alphaBoolBlend {blendOption=Linear}
    alphaScaleBiasClamp {bClampResult=true, clampMax=0.33, clampMin=0.1, scale=0.0033}
    boundBone {boneName=dyn_scarf_b}
    componentLinearAccScale=<1 1 1>
    componentLinearVelScale=<1 1 1>
    physicsBodyDefinitions [{
      sphereCollisionRadius=0
      boundBone {boneName=dyn_scarf_b}
      boxExtents=<5 5 10>
      constraintSetup {
       angularTargetAxis=AxisZ
       twistAxis=AxisZ
       angularLimitsMax=<0 0 0>
       angularLimitsMin=<0 0 0>
       angularTarget=<0 0 0>
       linearAxesMax=<4 4 4>
       linearAxesMin=<-4 -4 -4>
      }
      localJointOffset=<0 0 -5>
     }]
   }
   n8 = "Constraint - dyn_tags (PP)"!(ComponentPose=@n7, Alpha=1, ConstraintWeights_0=0.229, ConstraintWeights_1:float=0.771) -> (Pose) {
    alphaBoolBlend {blendOption=Linear}
    boneToModify {boneName=dyn_tags}
    constraintSetup [{targetBone {boneName=spine_04}}, {targetBone {boneName=spine_05}}]
    constraintWeights [1,1]
   }
   n9 = "Transform (Modify) Bone - Bone: dyn_tags"(ComponentPose=@n8, Alpha=1, Translation="0,0,0", Rotation="0.000,0.000,-15", Scale="1,1,1") {
    rotationMode=BMM_Additive
    rotationSpace=BCS_BoneSpace
    alphaBoolBlend {blendOption=Linear}
    boneToModify {boneName=dyn_tags}
    rotation {roll=-15}
   }
   n10 = Anim Dynamics - Bone: dyn_tags(ComponentPose=@n9, Alpha=0.7, ExternalForce="0,0,0") {
    angularBiasOverride=1
    angularDampingOverride=0.9999
    bGravityOverrideInSimSpace=true
    bOverrideAngularDamping=true
    lODThreshold=1
    numSolverIterationsPostUpdate=4
    numSolverIterationsPreUpdate=16
    alphaBoolBlend {blendOption=Linear}
    boundBone {boneName=dyn_tags}
    physicsBodyDefinitions [{
      sphereCollisionRadius=0
      boundBone {boneName=dyn_tags}
      boxExtents=<5 2 10>
      constraintSetup {angularTargetAxis=AxisZ, twistAxis=AxisZ, angularLimitsMax=<1 45 0>, angularLimitsMin=<-15 -45 0>, angularTarget=<0 0 0>}
      localJointOffset=<0 1 -1>
     }]
   }
   n11 = Anim Dynamics - Bone: dyn_boot_strings_01_l(ComponentPose=@n10, Alpha=0.55, ExternalForce="0,0,0") {
    bGravityOverrideInSimSpace=true
    bLinearSpring=true
    bOverrideLinearDamping=true
    gravityScale=0.33
    linearDampingOverride=0.995
    linearSpringConstant=150
    simulationSpace=RootRelative
    alphaBoolBlend {blendOption=Linear}
    boundBone {boneName=dyn_boot_strings_01_l}
    physicsBodyDefinitions [{
      sphereCollisionRadius=0
      boundBone {boneName=dyn_boot_strings_01_l}
      boxExtents=<10 10 10>
      constraintSetup {angularLimitsMax=<0 0 0>, angularLimitsMin=<0 0 0>, angularTarget=<0 0 0>, linearAxesMax=<4 2 0.02>, linearAxesMin=<-4 -2 -0.02>}
      localJointOffset=<0 0 -5>
     }]
   }
   n12 = Anim Dynamics - Bone: dyn_boot_strings_02_l(ComponentPose=@n11, Alpha=0.3, ExternalForce="0,0,0") {
    bGravityOverrideInSimSpace=true
    bLinearSpring=true
    bOverrideLinearDamping=true
    gravityScale=0.29
    linearDampingOverride=0.995
    linearSpringConstant=150
    simulationSpace=RootRelative
    alphaBoolBlend {blendOption=Linear}
    boundBone {boneName=dyn_boot_strings_02_l}
    physicsBodyDefinitions [{
      sphereCollisionRadius=0
      boundBone {boneName=dyn_boot_strings_02_l}
      boxExtents=<10 10 10>
      constraintSetup {angularLimitsMax=<0 0 0>, angularLimitsMin=<0 0 0>, angularTarget=<0 0 0>, linearAxesMax=<4 2 0.02>, linearAxesMin=<-5 -2 -0.02>}
      localJointOffset=<0 0 -6>
     }]
   }
   n13 = Anim Dynamics - Bone: dyn_boot_strings_01_r(ComponentPose=@n12, Alpha=0.55, ExternalForce="0,0,0") {
    bGravityOverrideInSimSpace=true
    bLinearSpring=true
    bOverrideLinearDamping=true
    gravityScale=0.33
    linearDampingOverride=0.995
    linearSpringConstant=150
    simulationSpace=RootRelative
    alphaBoolBlend {blendOption=Linear}
    boundBone {boneName=dyn_boot_strings_01_r}
    physicsBodyDefinitions [{
      sphereCollisionRadius=0
      boundBone {boneName=dyn_boot_strings_01_r}
      boxExtents=<10 10 10>
      constraintSetup {angularLimitsMax=<0 0 0>, angularLimitsMin=<0 0 0>, angularTarget=<0 0 0>, linearAxesMax=<4 2 0.02>, linearAxesMin=<-4 -2 -0.02>}
      localJointOffset=<0 0 -5>
     }]
   }
   n14 = Anim Dynamics - Bone: dyn_boot_strings_02_r(ComponentPose=@n13, Alpha=0.3, ExternalForce="0,0,0") {
    bGravityOverrideInSimSpace=true
    bLinearSpring=true
    bOverrideLinearDamping=true
    gravityScale=0.29
    linearDampingOverride=0.995
    linearSpringConstant=150
    simulationSpace=RootRelative
    alphaBoolBlend {blendOption=Linear}
    boundBone {boneName=dyn_boot_strings_02_r}
    physicsBodyDefinitions [{
      sphereCollisionRadius=0
      boundBone {boneName=dyn_boot_strings_02_r}
      boxExtents=<10 10 10>
      constraintSetup {angularLimitsMax=<0 0 0>, angularLimitsMin=<0 0 0>, angularTarget=<0 0 0>, linearAxesMax=<4 2 0.02>, linearAxesMin=<-5 -2 -0.02>}
      localJointOffset=<0 0 -6>
     }]
   }
   n15 = "Transform (Modify) Bone - Bone: upperarm_cuff_r"(ComponentPose=@n14, Alpha=Get arm_rot_r() -> (arm_rot_r:double), Translation="0,0,0", Rotation="35,0.000,0.000", Scale="1.25,1.25,1.25") {
    rotationMode=BMM_Additive
    rotationSpace=BCS_BoneSpace
    scaleMode=BMM_Additive
    scaleSpace=BCS_BoneSpace
    alphaBoolBlend {blendOption=Linear}
    boneToModify {boneName=upperarm_cuff_r}
    rotation {pitch=35}
    scale=<1.25 1.25 1.25>
   }
   n16 = "Transform (Modify) Bone - Bone: upperarm_cuff_l"(ComponentPose=@n15, Alpha=Get arm_rot_l() -> (arm_rot_l:double), Translation="0,0,0", Rotation="30,0.000,0.000", Scale="1,1,1") {
    rotationMode=BMM_Additive
    rotationSpace=BCS_BoneSpace
    alphaBoolBlend {blendOption=Linear}
    boneToModify {boneName=upperarm_cuff_l}
    rotation {pitch=30}
   }
   n17 = RigidBody(ComponentPose=@n16, Alpha=0.915) {
    bOverrideWorldGravity=true
    overridePhysicsAsset=&/F_MED_Raptor_Dyn_Jacket
    simulationSpace=BaseBoneSpace
    alphaBoolBlend {blendOption=Linear}
    baseBoneRef {boneName=spine_01}
    componentAppliedLinearAccClamp=<3000 3000 3000>
    componentLinearAccScale=<0.35 0.2 0.2>
    componentLinearVelScale=<0.35 0.2 0.2>
    overrideWorldGravity {z=-980}
   }
   n18 = RigidBody(ComponentPose=@n17, Alpha=0.15) {
    bOverrideWorldGravity=true
    overridePhysicsAsset=&/F_MED_Raptor_Dyn_Collar
    simulationSpace=BaseBoneSpace
    alphaBoolBlend {blendOption=Linear}
    baseBoneRef {boneName=spine_01}
    componentAppliedLinearAccClamp {z=100}
    componentLinearAccScale=<3 3 0.1>
    componentLinearVelScale=<3 3 0.1>
    overrideWorldGravity {z=-980}
   }
   Output Pose(Result=Component To Local(ComponentPose=@n18)) {layerGroup=None}
  ]}
 EventGraph {nodes [
   n1 = Event BlueprintUpdateAnimation()
   n2 = GetOwnerSkeletalMeshComponent()
   n3 = Set Skeletal_Mesh(execute=@n7.Is Valid, Skeletal_Mesh:Skeletal Mesh Component Object Reference=@n2) -> (Output_Get:Skeletal Mesh Component Object Reference)
   Is Valid(exec=@n1.then, InputObject=@n3)
   > n4 = Sequence()
   n5 = BreakTransform(InTransform=GetDeltaTransformFromRefPose(self=Get Skeletal_Mesh() -> (Skeletal_Mesh:Skeletal Mesh Component Object Reference), BoneName=lowerarm_r))
   > Set arm_rot_r(arm_rot_r:double=MapRangeClamped(Value=BreakRotator(InRot=@n5.Rotation).Yaw, InRangeA=30, InRangeB=90)) -> (Output_Get:double)
   n6 = BreakTransform(InTransform=GetDeltaTransformFromRefPose(self=Get Skeletal_Mesh(), BoneName=lowerarm_l))
   Set arm_rot_l(execute=@n4.then_1, arm_rot_l:double=MapRangeClamped(Value=BreakRotator(InRot=@n6.Rotation).Yaw, InRangeA=40, InRangeB=100)) -> (Output_Get:double)
   Event BlueprintInitializeAnimation()
   > n7 = Is Valid(InputObject=@n2)
   Event OnEndCrouching()
   > Set crouch_alpha(crouch_alpha:double=0) -> (Output_Get:double)
   Event OnBeginCrouching()
   > Set crouch_alpha(crouch_alpha=1)
  ]}
]
variables [
 arm_rot_l {type=number}
 arm_rot_r {type=number}
 crouch_alpha {type=number}
 Skeletal_Mesh {type=SkeletalMeshComponent}
]
